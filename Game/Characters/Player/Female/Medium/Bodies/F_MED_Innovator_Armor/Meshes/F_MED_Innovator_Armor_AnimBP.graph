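# Post-process style AnimBP for the female medium Innovator armor body, copying the parent mesh's pose and adding secondary motion and clothing fixes.
# AnimDynamics springs four pouch bones, then weighted multi-bone constraints position clips, gun and pouch bases from spine, pelvis and thighs.
# Interpolated additive offsets adjust shoulderpads and kneepads.
# The EventGraph drives these from FortPlayerAnimInstance: targeting pistol, rifle/shotgun weapon types, and crouching, jumping or falling.
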AnimBlueprint /Game/Characters/Player/Female/Medium/Bodies/F_MED_Innovator_Armor/Meshes/F_MED_Innovator_Armor_AnimBP
target_skeleton=&/F_MED_Innovator_Armor_Skeleton
graphs [
 AnimGraph {nodes [
   n1 = Anim Dynamics - Bone: dyn_pouch_03_parent(ComponentPose=Local To Component(LocalPose=Copy Pose From Mesh() {bUseAttachedParent=true}), Alpha=1, ExternalForce="0,0,0") {
    angularDampingOverride=0.9999
    bAngularSpring=true
    bGravityOverrideInSimSpace=true
    bLinearSpring=true
    bOverrideAngularDamping=true
    bOverrideLinearDamping=true
    bUseGravityOverride=true
    linearDampingOverride=0.999
    simulationSpace=BoneRelative
    alphaBoolBlend {blendOption=Linear}
    boundBone {boneName=dyn_pouch_03_parent}
    componentAppliedLinearAccClamp=<8000 8000 8000>
    componentLinearAccScale=<0.15 0.15 0.15>
    componentLinearVelScale=<0.15 0.15 0.15>
    gravityOverride=<1 1 1>
    physicsBodyDefinitions [{
      sphereCollisionRadius=0
      boundBone {boneName=dyn_pouch_03_parent}
      boxExtents=<12 15 4>
      constraintSetup {angularLimitsMax=<0 5 5>, angularLimitsMin=<0 -2 -5>, angularTarget=<0 0 0>}
      localJointOffset=<4.5 0 0>
     }]
    relativeSpaceBone {boneName=pelvis}
   }
   n2 = Anim Dynamics - Bone: dyn_pouch_02_parent(ComponentPose=@n1, Alpha=0.5, ExternalForce="0,0,0") {
    angularDampingOverride=0.9999
    bAngularSpring=true
    bGravityOverrideInSimSpace=true
    bLinearSpring=true
    bOverrideAngularDamping=true
    bOverrideLinearDamping=true
    bUseGravityOverride=true
    linearDampingOverride=0.9999
    simulationSpace=BoneRelative
    alphaBoolBlend {blendOption=Linear}
    boundBone {boneName=dyn_pouch_02_parent}
    componentAppliedLinearAccClamp=<6000 6000 6000>
    componentLinearAccScale=<0.15 0.15 0.15>
    componentLinearVelScale=<0.15 0.15 0.15>
    gravityOverride=<1 1 1>
    physicsBodyDefinitions [{
      sphereCollisionRadius=0
      boundBone {boneName=dyn_pouch_02_parent}
      boxExtents=<15 20 4>
      constraintSetup {angularLimitsMax=<0 5 5>, angularLimitsMin=<0 -2 -5>, angularTarget=<0 0 0>}
      localJointOffset=<4.5 0 0>
     }]
    relativeSpaceBone {boneName=root}
   }
   n3 = Anim Dynamics - Bone: dyn_pouch_04_parent(ComponentPose=@n2, Alpha=0.75, ExternalForce="0,0,0") {
    angularDampingOverride=0.9999
    bAngularSpring=true
    bGravityOverrideInSimSpace=true
    bLinearSpring=true
    bOverrideAngularDamping=true
    bOverrideLinearDamping=true
    bUseGravityOverride=true
    linearDampingOverride=0.9999
    simulationSpace=BoneRelative
    alphaBoolBlend {blendOption=Linear}
    boundBone {boneName=dyn_pouch_04_parent}
    componentAppliedLinearAccClamp=<6000 6000 6000>
    componentLinearAccScale=<0.15 0.15 0.15>
    componentLinearVelScale=<0.15 0.15 0.15>
    gravityOverride=<1 1 1>
    physicsBodyDefinitions [{
      sphereCollisionRadius=0
      boundBone {boneName=dyn_pouch_04_parent}
      boxExtents=<12 15 4>
      constraintSetup {angularLimitsMax=<0 5 5>, angularLimitsMin=<0 -2 -5>, angularTarget=<0 0 0>}
      localJointOffset=<4.5 0 0>
     }]
    relativeSpaceBone {boneName=root}
   }
   n4 = Anim Dynamics - Bone: dyn_pouch_01_parent(ComponentPose=@n3, Alpha=1, ExternalForce="0,0,0") {
    angularDampingOverride=0.9999
    bAngularSpring=true
    bGravityOverrideInSimSpace=true
    bLinearSpring=true
    bOverrideAngularDamping=true
    bOverrideLinearDamping=true
    bUseGravityOverride=true
    linearDampingOverride=0.9999
    simulationSpace=BoneRelative
    alphaBoolBlend {blendOption=Linear}
    boundBone {boneName=dyn_pouch_01_parent}
    componentAppliedLinearAccClamp=<6000 6000 6000>
    componentLinearAccScale=<0.15 0.15 0.15>
    componentLinearVelScale=<0.15 0.15 0.15>
    gravityOverride=<1 1 1>
    physicsBodyDefinitions [{
      sphereCollisionRadius=0
      boundBone {boneName=dyn_pouch_01_parent}
      boxExtents=<12 20 4>
      constraintSetup {angularLimitsMax=<0 5 5>, angularLimitsMin=<0 -2 -5>, angularTarget=<0 0 0>}
      localJointOffset=<4.5 0 0>
     }]
    relativeSpaceBone {boneName=root}
   }
   n5 = "Constraint - dyn_clip_01 (PPPP)"!(
    ComponentPose=@n4
    Alpha=1
    ConstraintWeights_0=0.2372
    ConstraintWeights_1:float=0.1636
    ConstraintWeights_2:float=0.0027
    ConstraintWeights_3:float=0.5965
   ) -> (Pose) {
    alphaBoolBlend {blendOption=Linear}
    boneToModify {boneName=dyn_clip_01}
    constraintSetup [
     {targetBone {boneName=thigh_r}}
     {targetBone {boneName=pelvis}}
     {targetBone {boneName=thigh_twist_01_l}}
     {targetBone {boneName=thigh_twist_01_r}}
    ]
    constraintWeights [1,1,1,1]
   }
   n6 = "Constraint - dyn_clip_02 (PPPP)"!(
    ComponentPose=@n5
    Alpha=1
    ConstraintWeights_0=0.225
    ConstraintWeights_1:float=0.215
    ConstraintWeights_2:float=0.0023
    ConstraintWeights_3:float=0.5577
   ) -> (Pose) {
    alphaBoolBlend {blendOption=Linear}
    boneToModify {boneName=dyn_clip_02}
    constraintSetup [
     {targetBone {boneName=thigh_r}}
     {targetBone {boneName=pelvis}}
     {targetBone {boneName=thigh_twist_01_l}}
     {targetBone {boneName=thigh_twist_01_r}}
    ]
    constraintWeights [1,1,1,1]
   }
   n7 = "Constraint - dyn_gun_r (PPP)"!(ComponentPose=@n6, Alpha=1, ConstraintWeights_0=0.1299, ConstraintWeights_1:float=0.5953, ConstraintWeights_2:float=0.2748) -> (Pose) {
    alphaBoolBlend {blendOption=Linear}
    boneToModify {boneName=dyn_gun_r}
    constraintSetup [{targetBone {boneName=pelvis}}, {targetBone {boneName=thigh_twist_01_l}}, {targetBone {boneName=thigh_l}}]
    constraintWeights [1,1,1]
   }
   n8 = "Constraint - dyn_pouch_01_base (PPPPP)"!(
    ComponentPose=@n7
    Alpha=1
    ConstraintWeights_0=0.1306
    ConstraintWeights_1:float=0.1786
    ConstraintWeights_2:float=0.0052
    ConstraintWeights_3:float=0.0238
    ConstraintWeights_4:float=0.6618
   ) -> (Pose) {
    alphaBoolBlend {blendOption=Linear}
    boneToModify {boneName=dyn_pouch_01_base}
    constraintSetup [
     {targetBone {boneName=spine_01}}
     {targetBone {boneName=spine_02}}
     {targetBone {boneName=spine_03}}
     {targetBone {boneName=spine_04}}
     {targetBone {boneName=pelvis}}
    ]
    constraintWeights [1,1,1,1,1]
   }
   n9 = "Constraint - dyn_pouch_02_base (PPPPP)"!(
    ComponentPose=@n8
    Alpha=1
    ConstraintWeights_0=0.1491
    ConstraintWeights_1:float=0.1894
    ConstraintWeights_2:float=0.009
    ConstraintWeights_3:float=0.0195
    ConstraintWeights_4:float=0.6331
   ) -> (Pose) {
    alphaBoolBlend {blendOption=Linear}
    boneToModify {boneName=dyn_pouch_02_base}
    constraintSetup [
     {targetBone {boneName=spine_01}}
     {targetBone {boneName=spine_02}}
     {targetBone {boneName=spine_03}}
     {targetBone {boneName=spine_04}}
     {targetBone {boneName=pelvis}}
    ]
    constraintWeights [1,1,1,1,1]
   }
   n10 = "Constraint - dyn_pouch_03_base (PPPPP)"!(
    ComponentPose=@n9
    Alpha=1
    ConstraintWeights_0=0.152
    ConstraintWeights_1:float=0.1906
    ConstraintWeights_2:float=0.0091
    ConstraintWeights_3:float=0.0194
    ConstraintWeights_4:float=0.629
   ) -> (Pose) {
    alphaBoolBlend {blendOption=Linear}
    boneToModify {boneName=dyn_pouch_03_base}
    constraintSetup [
     {targetBone {boneName=spine_01}}
     {targetBone {boneName=spine_02}}
     {targetBone {boneName=spine_03}}
     {targetBone {boneName=spine_04}}
     {targetBone {boneName=pelvis}}
    ]
    constraintWeights [1,1,1,1,1]
   }
   n11 = "Constraint - dyn_pouch_04_base (PPPPP)"!(
    ComponentPose=@n10
    Alpha=1
    ConstraintWeights_0=0.1342
    ConstraintWeights_1:float=0.1798
    ConstraintWeights_2:float=0.0053
    ConstraintWeights_3:float=0.0227
    ConstraintWeights_4:float=0.6579
   ) -> (Pose) {
    alphaBoolBlend {blendOption=Linear}
    boneToModify {boneName=dyn_pouch_04_base}
    constraintSetup [
     {targetBone {boneName=spine_01}}
     {targetBone {boneName=spine_02}}
     {targetBone {boneName=spine_03}}
     {targetBone {boneName=spine_04}}
     {targetBone {boneName=pelvis}}
    ]
    constraintWeights [1,1,1,1,1]
   }
   n12 = Get Set_Pistol() -> (Set_Pistol:double)
   n13 = "Transform (Modify) Bone - Bone: dyn_shoulderpad_l"(ComponentPose=@n11, Alpha=@n12, Translation="-1,-0.000,-0.5", Rotation="-15,0,10", Scale="1,1,1") {
    rotationMode=BMM_Additive
    rotationSpace=BCS_BoneSpace
    translationMode=BMM_Additive
    translationSpace=BCS_BoneSpace
    alphaBoolBlend {blendOption=Linear}
    alphaScaleBiasClamp {bInterpResult=true, interpSpeedDecreasing=4, interpSpeedIncreasing=6}
    boneToModify {boneName=dyn_shoulderpad_l}
    rotation {pitch=-15, roll=10}
    translation {x=-1, z=-0.5}
   }
   n14 = "Transform (Modify) Bone - Bone: dyn_shoulderpad_r"(ComponentPose=@n13, Alpha=@n12, Translation="-0.000,-2,0", Rotation="-15,0,25", Scale="1,1,1") {
    rotationMode=BMM_Additive
    rotationSpace=BCS_BoneSpace
    translationMode=BMM_Additive
    translationSpace=BCS_BoneSpace
    alphaBoolBlend {blendOption=Linear}
    alphaScaleBiasClamp {bInterpResult=true, interpSpeedDecreasing=4, interpSpeedIncreasing=6}
    boneToModify {boneName=dyn_shoulderpad_r}
    rotation {pitch=-15, roll=25}
    translation {y=-2}
   }
   n15 = Get Set_Shotgun() -> (Set_Shotgun:double)
   n16 = "Transform (Modify) Bone - Bone: dyn_shoulderpad_r"(ComponentPose=@n14, Alpha=@n15, Translation="2,-1.000,0.000", Rotation="0,-0.000,20", Scale="1,1,1") {
    rotationMode=BMM_Additive
    rotationSpace=BCS_BoneSpace
    translationMode=BMM_Additive
    translationSpace=BCS_BoneSpace
    alphaBoolBlend {blendOption=Linear}
    alphaScaleBiasClamp {bInterpResult=true, interpSpeedDecreasing=4, interpSpeedIncreasing=6}
    boneToModify {boneName=dyn_shoulderpad_r}
    rotation {roll=20}
    translation=<2 -1>
   }
   n17 = "Transform (Modify) Bone - Bone: dyn_shoulderpad_l"(ComponentPose=@n16, Alpha=@n15, Translation="1,0.000,-0.25", Rotation="0,0,0", Scale="1,1,1") {
    rotationMode=BMM_Additive
    rotationSpace=BCS_BoneSpace
    translationMode=BMM_Additive
    translationSpace=BCS_BoneSpace
    alphaBoolBlend {blendOption=Linear}
    alphaScaleBiasClamp {bInterpResult=true, interpSpeedDecreasing=4, interpSpeedIncreasing=6}
    boneToModify {boneName=dyn_shoulderpad_l}
    translation {x=1, z=-0.25}
   }
   n18 = Get Is_Crouching() -> (Is_Crouching:double)
   n19 = "Transform (Modify) Bone - Bone: dyn_kneepad_parent_l"(ComponentPose=@n17, Alpha=@n18, Translation="0,0,2.5", Rotation="0,0,0", Scale="1,1,1") {
    rotationMode=BMM_Additive
    rotationSpace=BCS_BoneSpace
    translationMode=BMM_Additive
    translationSpace=BCS_BoneSpace
    alphaBoolBlend {blendOption=Linear}
    alphaScaleBiasClamp {bInterpResult=true}
    boneToModify {boneName=dyn_kneepad_parent_l}
    translation {z=2.5}
   }
   n20 = "Transform (Modify) Bone - Bone: dyn_kneepad_parent_r"(ComponentPose=@n19, Alpha=@n18, Translation="0,0,2.5", Rotation="0,0,0", Scale="1,1,1") {
    rotationMode=BMM_Additive
    rotationSpace=BCS_BoneSpace
    translationMode=BMM_Additive
    translationSpace=BCS_BoneSpace
    alphaBoolBlend {blendOption=Linear}
    alphaScaleBiasClamp {bInterpResult=true}
    boneToModify {boneName=dyn_kneepad_parent_r}
    translation {z=2.5}
   }
   Output Pose(Result=Component To Local(ComponentPose=@n20)) {layerGroup=None}
  ]}
 EventGraph {nodes [
   Event BlueprintUpdateAnimation()
   > n1 = Sequence!(execute) -> (then_0, then_1, then_2:Exec, then_3:Exec)
   > n2 = Cast To FortPlayerAnimInstance(Object=GetAnimInstance(self=GetOwnerSkeletalMeshComponent()))
   > Set Weapon_Pistols(Weapon_Pistols:bool=false) -> (Output_Get:bool)
   > Set Weapon_Shotguns(Weapon_Shotguns:bool=false) -> (Output_Get:bool)
   > n3 = Switch on EFortWeaponCoreAnimation!(execute, Selection:EFortWeaponCoreAnimation Enum=Get WeaponCoreAnim(self=@n2)) -> (
    Melee:Exec
    Pistol:Exec
    Shotgun:Exec
    PaperBlueprint:Exec
    Rifle:Exec
    MeleeOneHand:Exec
    MachinePistol:Exec
    RocketLauncher:Exec
    GrenadeLauncher:Exec
    GoingCommando:Exec
    AssaultRifle:Exec
    TacticalShotgun:Exec
    SniperRifle:Exec
    TrapPlacement:Exec
    ShoulderLauncher:Exec
    AbilityDecoTool:Exec
    Crossbow:Exec
    C4:Exec
    RemoteControl:Exec
    DualWield:Exec
    AR_BullPup:Exec
    AR_ForwardGrip:Exec
    MedPackPaddles:Exec
    SMG_P90:Exec
    AR_DrumGun:Exec
    Consumable_Small:Exec
    Consumable_Large:Exec
    Balloon:Exec
    MountedTurret:Exec
    CreativeTool:Exec
    ExplosiveBow:Exec
    AshtonIndigo:Exec
    AshtonChicago:Exec
    MeleeDualWield:Exec
    Unarmed:Exec
   )
   Branch(execute=[@n3.AssaultRifle, @n3.Rifle, @n3.Shotgun, @n3.SniperRifle, @n3.TacticalShotgun])
   > Set Weapon_Shotguns(Weapon_Shotguns=true)
   Branch(execute=@n3.Pistol, Condition=Get bIsTargeting())
   > Set Weapon_Pistols(Weapon_Pistols=true)
   Set Set_Shotgun(execute=@n1.then_2, Set_Shotgun:double=SelectFloat(A=1, B=0, bPickA=Get Weapon_Shotguns() -> (Weapon_Shotguns:bool))) -> (Output_Get:double)
   Set Set_Pistol(execute=@n1.then_1, Set_Pistol:double=SelectFloat(A=1, B=0, bPickA=Get Weapon_Pistols() -> (Weapon_Pistols:bool))) -> (Output_Get:double)
   n4 = Get bIsCrouching()
   n5 = Get bIsJumping()
   n6 = Get bIsFalling()
   Set Is_Crouching(execute=@n1.then_3, Is_Crouching:double=SelectFloat(A=1, B=0, bPickA=OR Boolean!(A=@n4, B=@n5, C:bool=@n6) -> (ReturnValue))) -> (Output_Get:double)
   Get bIsTargeting()
  ]}
]
variables [
 Is_Crouching {type=number}
 Set_Pistol {type=number}
 Set_Shotgun {type=number}
 Weapon_Pistols {type=boolean}
 Weapon_Shotguns {type=boolean}
]
