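# Physics asset for the Innovator Armor face accessory on the medium female player body, defining bodies and joints for dynamic hair bones.
# Capsule bodies cover hair strands in front, back, left and right groups, each constrained to its group master with narrow swing and twist limits.
# Only the back master is constrained to the root hair master, almost rigidly.
# The head, root master and other group masters are small spheres without parent constraints.
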PhysicsAsset /Game/Characters/Player/Female/Medium/Bodies/F_MED_Innovator_Armor/Meshes/Parts/F_MED_Innovator_Armor_FaceAcc_Physics
bodies {
 dyn_hair_01_bk [{shapeType=Capsule, length=2.0406, radius=1.8253, center=<1.4145 0 0>, rotation=<rot 90 0 0>}]
 dyn_hair_01_frnt [{shapeType=Capsule, length=2.2155, radius=1.0746, center=<1.8513 0 -0.0001>, rotation=<rot 90 0 0>}]
 dyn_hair_01_l [{shapeType=Capsule, length=1.4353, radius=1.666, center=<1.6405 0 0>, rotation=<rot 90 0 0>}]
 dyn_hair_01_r [{shapeType=Capsule, length=1.5148, radius=1.9899, center=<1.6293 0 -0.0002>, rotation=<rot 90 0 0>}]
 dyn_hair_02_bk [{shapeType=Capsule, length=2.9396, radius=1.5739, center=<1.9491 0 0>, rotation=<rot 90 0 0>}]
 dyn_hair_02_frnt [{shapeType=Capsule, length=1.5674, radius=1.7933, center=<2.2879 0 0.0001>, rotation=<rot 90 0 0>}]
 dyn_hair_02_l [{shapeType=Capsule, length=2.8005, radius=1.393, center=<2.439 0 0>, rotation=<rot 90 0 0>}]
 dyn_hair_02_r [{shapeType=Capsule, length=3.6197, radius=1.7447, center=<3.0482 0 0>, rotation=<rot 90 0 0>}]
 dyn_hair_03_bk [{shapeType=Capsule, length=1.5181, radius=2.1465, center=<2.0202 0 0>, rotation=<rot 90 0 0>}]
 dyn_hair_03_frnt [{shapeType=Capsule, length=2.7499, radius=1.229, center=<1.7058 0 0.0004>, rotation=<rot 90 0 0>}]
 dyn_hair_03_l [{shapeType=Capsule, length=1.4301, radius=1.5995, center=<2.4831 0 0>, rotation=<rot 90 0 0>}]
 dyn_hair_03_r [{shapeType=Capsule, length=1.9869, radius=1.1291, center=<1.8287 0 0>, rotation=<rot 90 0 0>}]
 dyn_hair_04_bk [{shapeType=Capsule, length=2.2272, radius=1.4778, center=<1.3544 0 0>, rotation=<rot 90 0 0>}]
 dyn_hair_04_frnt [{shapeType=Capsule, length=2.2936, radius=1.3689, center=<2.0645 0 0>, rotation=<rot 90 0 0>}]
 dyn_hair_05_bk [{shapeType=Capsule, length=1.4816, radius=1.4478, center=<1.4727 0 -0.0001>, rotation=<rot 90 0 0>}]
 dyn_hair_06_bk [{shapeType=Capsule, length=2.4865, radius=1.6298, center=<2.2998 0 0>, rotation=<rot 90 0 0>}]
 dyn_hair_07_bk [{shapeType=Capsule, length=1.4971, radius=1.3225, center=<1.7246 0 0>, rotation=<rot 90 0 0>}]
 dyn_hair_08_bk [{shapeType=Capsule, length=1.796, radius=1.6601, center=<1.7512 0 0.0001>, rotation=<rot 90 0 0>}]
 dyn_hair_master [{shapeType=Sphere, radius=0.5}]
 dyn_hair_master_bk [{shapeType=Capsule, length=2.2169, radius=1.7631, center=<0 0 0.6948>, rotation=<rot 0 0 90>}]
 dyn_hair_master_frnt [{shapeType=Sphere, radius=0.5}]
 dyn_hair_master_l [{shapeType=Sphere, radius=0.5}]
 dyn_hair_master_r [{shapeType=Sphere, radius=0.5}]
 head [{shapeType=Sphere, radius=0.5}]
}
constraints [
 {bone1Name=dyn_hair_01_r, bone2Name=dyn_hair_master_r, swing1LimitDegrees=20, swing2LimitDegrees=20, twistLimitDegrees=3}
 {bone1Name=dyn_hair_02_r, bone2Name=dyn_hair_master_r, swing1LimitDegrees=10, swing2LimitDegrees=10, twistLimitDegrees=3}
 {bone1Name=dyn_hair_03_r, bone2Name=dyn_hair_master_r, swing1LimitDegrees=10, swing2LimitDegrees=10, twistLimitDegrees=3}
 {bone1Name=dyn_hair_03_l, bone2Name=dyn_hair_master_l, swing1LimitDegrees=15, swing2LimitDegrees=15, twistLimitDegrees=3}
 {bone1Name=dyn_hair_02_l, bone2Name=dyn_hair_master_l, swing1LimitDegrees=10, swing2LimitDegrees=10, twistLimitDegrees=3}
 {bone1Name=dyn_hair_01_l, bone2Name=dyn_hair_master_l, swing1LimitDegrees=20, swing2LimitDegrees=20, twistLimitDegrees=3}
 {bone1Name=dyn_hair_01_frnt, bone2Name=dyn_hair_master_frnt, swing1LimitDegrees=20, swing2LimitDegrees=20, twistLimitDegrees=3}
 {bone1Name=dyn_hair_02_frnt, bone2Name=dyn_hair_master_frnt, swing1LimitDegrees=20, swing2LimitDegrees=20, twistLimitDegrees=3}
 {bone1Name=dyn_hair_03_frnt, bone2Name=dyn_hair_master_frnt, swing1LimitDegrees=20, swing2LimitDegrees=20, twistLimitDegrees=3}
 {bone1Name=dyn_hair_04_frnt, bone2Name=dyn_hair_master_frnt, swing1LimitDegrees=20, swing2LimitDegrees=20, twistLimitDegrees=3}
 {bone1Name=dyn_hair_01_bk, bone2Name=dyn_hair_master_bk, swing1LimitDegrees=10, swing2LimitDegrees=10, twistLimitDegrees=3}
 {bone1Name=dyn_hair_02_bk, bone2Name=dyn_hair_master_bk, swing1LimitDegrees=15, swing2LimitDegrees=15, twistLimitDegrees=3}
 {bone1Name=dyn_hair_03_bk, bone2Name=dyn_hair_master_bk, swing1LimitDegrees=15, swing2LimitDegrees=15, twistLimitDegrees=3}
 {bone1Name=dyn_hair_04_bk, bone2Name=dyn_hair_master_bk, swing1LimitDegrees=15, swing2LimitDegrees=15, twistLimitDegrees=3}
 {bone1Name=dyn_hair_05_bk, bone2Name=dyn_hair_master_bk, swing1LimitDegrees=15, swing2LimitDegrees=15, twistLimitDegrees=3}
 {bone1Name=dyn_hair_06_bk, bone2Name=dyn_hair_master_bk, swing1LimitDegrees=5, swing2LimitDegrees=5, twistLimitDegrees=2}
 {bone1Name=dyn_hair_07_bk, bone2Name=dyn_hair_master_bk, swing1LimitDegrees=5, swing2LimitDegrees=5, twistLimitDegrees=2}
 {bone1Name=dyn_hair_08_bk, bone2Name=dyn_hair_master_bk, swing1LimitDegrees=15, swing2LimitDegrees=15, twistLimitDegrees=3}
 {bone1Name=dyn_hair_master_bk, bone2Name=dyn_hair_master, swing1LimitDegrees=1, swing2LimitDegrees=1, twistLimitDegrees=1}
]
