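# Animation Blueprint for the face accessory mesh of the female medium Innovator Armor body, adding physics-driven secondary motion.
# It copies the pose from the attached parent mesh, then runs a RigidBody node with an override physics asset.
# Simulation runs in spine_05 bone space with custom gravity and damped, clamped component motion.
# Jump and fall events lower and restore JumpAlpha, which the shown AnimGraph never reads.
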AnimBlueprint /Game/Characters/Player/Female/Medium/Bodies/F_MED_Innovator_Armor/Meshes/Parts/F_MED_Innovator_Armor_FaceAcc_AnimBP
target_skeleton=&/F_MED_Innovator_Armor_FaceAcc_Skeleton
graphs [
 AnimGraph {nodes [
   n1 = RigidBody(ComponentPose=Local To Component(LocalPose=Copy Pose From Mesh() {bUseAttachedParent=true})) {
    bClampLinearTranslationLimitToRefPose=true
    bOverrideWorldGravity=true
    overridePhysicsAsset=&/F_MED_Innovator_Armor_FaceAcc_Physics
    simulationSpace=BaseBoneSpace
    alphaBoolBlend {blendOption=Linear}
    baseBoneRef {boneName=spine_05}
    componentAppliedLinearAccClamp=<3000 3000 3000>
    componentLinearAccScale=<0.1 0.1 0.1>
    componentLinearVelScale=<0.1 0.1 0.1>
    overrideWorldGravity {z=-980}
   }
   Output Pose(Result=Component To Local(ComponentPose=@n1)) {layerGroup=None}
  ]}
 EventGraph {nodes [
   Event OnBeginJumping()
   > Set JumpAlpha(JumpAlpha:double=0.35) -> (Output_Get:double)
   Event OnEndJumping()
   > Set JumpAlpha(JumpAlpha=1)
   Event OnBeginFalling()
   > Set JumpAlpha(JumpAlpha=0.35)
   Event OnEndFalling()
   > Set JumpAlpha(JumpAlpha=1)
   Event BlueprintUpdateAnimation()
  ]}
]
variables [JumpAlpha {type=number}]
